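# Editor utility Blueprint that toggles the Can Be Damaged flag on selected building actors in the editor.
# Enabled and Disabled functions pass true or false to a shared Set Can Be Damaged function, which loops over the selection set.
# A chain of fallback casts matches BuildingProp, PropCorner, PropWall, Floor and Roof types, skipping anything else.
# Each match calls TransactObject before the flag changes, recording it for undo.
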EditorUtilityBlueprint /Game/Blueprints/Blutilities/UEFNToolkit/BuildingProp_ActorActions
graphs [
 Disabled {nodes [
   Disabled() {is=title K2Node_FunctionEntry}
   > Set Can Be Damaged!(execute, self, Can Be Damaged:bool=false) -> (then) {is=title K2Node_CallFunction}
  ]}
 Enabled {nodes [
   Enabled() {is=title K2Node_FunctionEntry}
   > Set Can Be Damaged(Can Be Damaged=true)
  ]}
 EventGraph {links [], nodes []}
 Set Can Be Damaged {nodes [
   n1 = Set Can Be Damaged!() -> (then, Can Be Damaged:bool=~) {is=title K2Node_FunctionEntry}
   > n2 = GetSelectionSet!(execute) -> (then, ReturnValue:Array of Actor Object References)
   > n3 = For Each Loop(Array:Array of Object References=@n2) -> (Array Element:Object Reference)
   > n4 = Cast To BuildingProp!(execute, Object=@n3) -> (then, CastFailed, AsBuilding Prop:Building Prop Object Reference) {is=title K2Node_DynamicCast}
   > TransactObject(Object=@n4)
   > SetCanBeDamaged!(execute, Actor:Actor Object Reference=@n4, bCanBeDamaged:bool=@n1) -> (then)
   n5 = Cast To BuildingPropCorner!(execute=@n4.CastFailed, Object=@n3) -> (then, CastFailed, AsBuilding Prop Corner:Building Prop Corner Object Reference) {is=title K2Node_DynamicCast}
   > TransactObject(Object=@n5)
   > SetCanBeDamaged(Actor=@n5, bCanBeDamaged=@n1)
   n6 = Cast To BuildingPropWall!(execute=@n5.CastFailed, Object=@n3) -> (then, CastFailed, AsBuilding Prop Wall:Building Prop Wall Object Reference) {is=title K2Node_DynamicCast}
   > TransactObject(Object=@n6)
   > SetCanBeDamaged(Actor=@n6, bCanBeDamaged=@n1)
   n7 = Cast To BuildingFloor!(execute=@n6.CastFailed, Object=@n3) -> (then, CastFailed, AsBuilding Floor:Building Floor Object Reference) {is=title K2Node_DynamicCast}
   > TransactObject(Object=@n7)
   > SetCanBeDamaged(Actor=@n7, bCanBeDamaged=@n1)
   n8 = Cast To BuildingRoof!(execute=@n7.CastFailed, Object=@n3) -> (then, CastFailed, AsBuilding Roof:Building Roof Object Reference) {is=title K2Node_DynamicCast}
   > TransactObject(Object=@n8)
   > SetCanBeDamaged(Actor=@n8, bCanBeDamaged=@n1)
  ]}
]
variables []
