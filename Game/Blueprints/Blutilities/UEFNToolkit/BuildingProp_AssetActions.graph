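# Editor utility blueprint of asset actions that batch-edit selected building prop Blueprints: Enable/Disable toggles damageability, Wood/Stone/Metal sets resource type.
# Shared functions resolve each asset's first subobject via SubobjectDataSubsystem, then try casts to BuildingProp, Corner, Wall, Floor and Roof.
# The resource path also rewrites AttributeInitKeys to the matching category through SetEditorProperty, inside a transaction with undo copies.
# The damage path calls TransactObject before each edit.
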EditorUtilityBlueprint /Game/Blueprints/Blutilities/UEFNToolkit/BuildingProp_AssetActions
graphs [
 Disable {nodes [
   Disable() {is=title K2Node_FunctionEntry}
   > Set Can Be Damaged!(execute, self, CanBeDamaged:bool=false) -> (then) {is=title K2Node_CallFunction}
  ]}
 Enable {nodes [
   Enable() {is=title K2Node_FunctionEntry}
   > Set Can Be Damaged(CanBeDamaged=true)
  ]}
 EventGraph {links [], nodes []}
 Metal {nodes [
   Metal() {is=title K2Node_FunctionEntry}
   > Set Prop Resource!(execute, self, Resource:E_AssetToProp_ResourceType Enum=NewEnumerator2) -> (then) {is=title K2Node_CallFunction}
  ]}
 Set Can Be Damaged {nodes [
   n1 = Set Can Be Damaged!() -> (then, CanBeDamaged:bool=~) {is=title K2Node_FunctionEntry}
   > n2 = GetSelectedAssets()
   > n3 = For Each Loop(Array:Array of Object References=@n2) -> (Array Element:Object Reference)
   > n4 = GetBlueprintAsset(Object=@n3)
   > n5 = Gather Subobject Data For Blueprint(self=Get SubobjectDataSubsystem(), Context=@n4)
   > n6 = GetObjectForBlueprint!(
    execute
    Data:"Subobject Data Structure (by ref)"=GetData(DataHandle="Get (a copy)"(Array:Array of Subobject Data Handle Structures=@n5) -> (Output:Subobject Data Handle Structure))
    Blueprint:Blueprint Object Reference=@n4
   ) -> (then, ReturnValue:Object Reference)
   > n7 = Cast To BuildingProp!(execute, Object=@n6) -> (then, CastFailed, AsBuilding Prop:Building Prop Object Reference) {is=title K2Node_DynamicCast}
   > TransactObject(Object=@n7)
   > SetCanBeDamaged!(execute, Actor:Actor Object Reference=@n7, bCanBeDamaged:bool=@n1) -> (then)
   n8 = Cast To BuildingPropCorner!(execute=@n7.CastFailed, Object=@n6) -> (then, CastFailed, AsBuilding Prop Corner:Building Prop Corner Object Reference) {is=title K2Node_DynamicCast}
   > TransactObject(Object=@n8)
   > SetCanBeDamaged(Actor=@n8, bCanBeDamaged=@n1)
   n9 = Cast To BuildingPropWall!(execute=@n8.CastFailed, Object=@n6) -> (then, CastFailed, AsBuilding Prop Wall:Building Prop Wall Object Reference) {is=title K2Node_DynamicCast}
   > TransactObject(Object=@n9)
   > SetCanBeDamaged(Actor=@n9, bCanBeDamaged=@n1)
   n10 = Cast To BuildingFloor!(execute=@n9.CastFailed, Object=@n6) -> (then, CastFailed, AsBuilding Floor:Building Floor Object Reference) {is=title K2Node_DynamicCast}
   > TransactObject(Object=@n10)
   > SetCanBeDamaged(Actor=@n10, bCanBeDamaged=@n1)
   n11 = Cast To BuildingRoof!(execute=@n10.CastFailed, Object=@n6) -> (then, CastFailed, AsBuilding Roof:Building Roof Object Reference) {is=title K2Node_DynamicCast}
   > TransactObject(Object=@n11)
   > SetCanBeDamaged(Actor=@n11, bCanBeDamaged=@n1)
  ]}
 Set Prop Resource {nodes [
   n1 = Set Prop Resource!() -> (then, Resource:E_AssetToProp_ResourceType Enum=NewEnumerator0) {is=title K2Node_FunctionEntry}
   > BeginTransaction()
   > n2 = Switch on E_AssetToProp_ResourceType!(execute, Selection:E_AssetToProp_ResourceType Enum=@n1) -> (NewEnumerator0:Exec, NewEnumerator1:Exec, NewEnumerator2:Exec) {is=title K2Node_SwitchEnum}
   > n3 = Set SelectedResource(SelectedResource:EFortResourceType Enum=Wood) -> (Output_Get:EFortResourceType Enum=Wood)
   n4 = GetSelectedAssets(execute=[@n3.then, @n50.then, @n49.then])
   > n5 = For Each Loop(Array=@n4)
   EndTransaction(execute=@n5.Completed)
   n6 = GetBlueprintAsset(execute=@n5.LoopBody, Object=@n5)
   > n7 = Gather Subobject Data For Blueprint(self=Get SubobjectDataSubsystem(), Context=@n6)
   > n8 = GetObjectForBlueprint(Data=GetData(DataHandle="Get (a copy)"(Array=@n7)), Blueprint=@n6)
   > CreateCopyForUndoBuffer!(execute, ObjectToModify:Object Reference=@n8) -> (then)
   > n9 = Cast To BuildingProp(Object=@n8)
   n10 = Get SelectedResource() -> (SelectedResource:EFortResourceType Enum=Wood)
   > SetResourceType!(execute, self:Building Static Mesh Object Reference=@n9, NewResourceType:EFortResourceType Enum=@n10) -> (then)
   > n11 = Switch on EFortResourceType!(execute, Selection:EFortResourceType Enum=@n10) -> (Wood:Exec, Stone:Exec, Metal:Exec, Permanite:Exec, GoldCurrency:Exec, Ingredient:Exec, None:Exec) {is=title K2Node_SwitchEnum}
   n12 = Set AttributeInitCategoryA(execute=[@n11.GoldCurrency, @n11.Ingredient, @n11.None, @n11.Permanite], AttributeInitCategoryA:Name=ResourceWood) -> (Output_Get:Name=None)
   n13 = Make Fort Attribute Initialization Key!(AttributeInitCategory:Name=Get AttributeInitCategoryA() -> (AttributeInitCategoryA:Name=None), AttributeInitSubCategory:Name=2) -> (FortAttributeInitializationKey:Fort Attribute Initialization Key Structure) {is=title K2Node_MakeStruct}
   SetEditorProperty(
    execute=[@n16.then, @n15.then, @n14.then, @n12.then]
    Object=@n9
    PropertyName=AttributeInitKeys
    PropertyValue:Array of Fort Attribute Initialization Key Structures=Make Array!("[0]":Fort Attribute Initialization Key Structure=@n13, "[1]":Fort Attribute Initialization Key Structure=@n13) -> (Array:Array of Fort Attribute Initialization Key Structures)
   )
   n14 = Set AttributeInitCategoryA(execute=@n11.Metal, AttributeInitCategoryA=ResourceMetal) -> (Output_Get=None)
   n15 = Set AttributeInitCategoryA(execute=@n11.Stone, AttributeInitCategoryA=ResourceStone) -> (Output_Get=None)
   n16 = Set AttributeInitCategoryA(execute=@n11.Wood, AttributeInitCategoryA=ResourceWood) -> (Output_Get=None)
   n17 = Cast To BuildingPropCorner(execute=@n9.CastFailed, Object=@n8)
   n18 = Get SelectedResource() -> (SelectedResource=Wood)
   > SetResourceType(self=@n17, NewResourceType=@n18)
   > n19 = Switch on EFortResourceType(Selection=@n18)
   n20 = Set AttributeInitCategoryA(execute=[@n19.GoldCurrency, @n19.Ingredient, @n19.None, @n19.Permanite], AttributeInitCategoryA=ResourceWood) -> (Output_Get=None)
   n21 = Make Fort Attribute Initialization Key(AttributeInitCategory=Get AttributeInitCategoryA() -> (AttributeInitCategoryA=None), AttributeInitSubCategory=2)
   SetEditorProperty(
    execute=[@n24.then, @n23.then, @n22.then, @n20.then]
    Object=@n17
    PropertyName=AttributeInitKeys
    PropertyValue=Make Array("[0]"=@n21, "[1]"=@n21)
   )
   n22 = Set AttributeInitCategoryA(execute=@n19.Metal, AttributeInitCategoryA=ResourceMetal) -> (Output_Get=None)
   n23 = Set AttributeInitCategoryA(execute=@n19.Stone, AttributeInitCategoryA=ResourceStone) -> (Output_Get=None)
   n24 = Set AttributeInitCategoryA(execute=@n19.Wood, AttributeInitCategoryA=ResourceWood) -> (Output_Get=None)
   n25 = Cast To BuildingPropWall(execute=@n17.CastFailed, Object=@n8)
   n26 = Get SelectedResource() -> (SelectedResource=Wood)
   > SetResourceType(self=@n25, NewResourceType=@n26)
   > n27 = Switch on EFortResourceType(Selection=@n26)
   n28 = Set AttributeInitCategoryA(execute=[@n27.GoldCurrency, @n27.Ingredient, @n27.None, @n27.Permanite], AttributeInitCategoryA=ResourceWood) -> (Output_Get=None)
   n29 = Make Fort Attribute Initialization Key(AttributeInitCategory=Get AttributeInitCategoryA() -> (AttributeInitCategoryA=None), AttributeInitSubCategory=2)
   SetEditorProperty(
    execute=[@n31.then, @n30.then, @n28.then, @n32.then]
    Object=@n25
    PropertyName=AttributeInitKeys
    PropertyValue=Make Array("[0]"=@n29, "[1]"=@n29)
   )
   n30 = Set AttributeInitCategoryA(execute=@n27.Metal, AttributeInitCategoryA=ResourceMetal) -> (Output_Get=None)
   n31 = Set AttributeInitCategoryA(execute=@n27.Stone, AttributeInitCategoryA=ResourceStone) -> (Output_Get=None)
   n32 = Set AttributeInitCategoryA(execute=@n27.Wood, AttributeInitCategoryA=ResourceWood) -> (Output_Get=None)
   n33 = Cast To BuildingFloor(execute=@n25.CastFailed, Object=@n8)
   n34 = Get SelectedResource() -> (SelectedResource=Wood)
   > SetResourceType(self=@n33, NewResourceType=@n34)
   > n35 = Switch on EFortResourceType(Selection=@n34)
   n36 = Set AttributeInitCategoryA(execute=[@n35.GoldCurrency, @n35.Ingredient, @n35.None, @n35.Permanite], AttributeInitCategoryA=ResourceWood) -> (Output_Get=None)
   n37 = Make Fort Attribute Initialization Key(AttributeInitCategory=Get AttributeInitCategoryA() -> (AttributeInitCategoryA=None), AttributeInitSubCategory=2)
   SetEditorProperty(
    execute=[@n40.then, @n39.then, @n38.then, @n36.then]
    Object=@n33
    PropertyName=AttributeInitKeys
    PropertyValue=Make Array("[0]"=@n37, "[1]"=@n37)
   )
   n38 = Set AttributeInitCategoryA(execute=@n35.Metal, AttributeInitCategoryA=ResourceMetal) -> (Output_Get=None)
   n39 = Set AttributeInitCategoryA(execute=@n35.Stone, AttributeInitCategoryA=ResourceStone) -> (Output_Get=None)
   n40 = Set AttributeInitCategoryA(execute=@n35.Wood, AttributeInitCategoryA=ResourceWood) -> (Output_Get=None)
   n41 = Cast To BuildingRoof(execute=@n33.CastFailed, Object=@n8)
   n42 = Get SelectedResource() -> (SelectedResource=Wood)
   > SetResourceType(self=@n41, NewResourceType=@n42)
   > n43 = Switch on EFortResourceType(Selection=@n42)
   n44 = Set AttributeInitCategoryA(execute=[@n43.GoldCurrency, @n43.Ingredient, @n43.None, @n43.Permanite], AttributeInitCategoryA=ResourceWood) -> (Output_Get=None)
   n45 = Make Fort Attribute Initialization Key(AttributeInitCategory=Get AttributeInitCategoryA() -> (AttributeInitCategoryA=None), AttributeInitSubCategory=2)
   SetEditorProperty(
    execute=[@n47.then, @n44.then, @n46.then, @n48.then]
    Object=@n41
    PropertyName=AttributeInitKeys
    PropertyValue=Make Array("[0]"=@n45, "[1]"=@n45)
   )
   n46 = Set AttributeInitCategoryA(execute=@n43.Metal, AttributeInitCategoryA=ResourceMetal) -> (Output_Get=None)
   n47 = Set AttributeInitCategoryA(execute=@n43.Stone, AttributeInitCategoryA=ResourceStone) -> (Output_Get=None)
   n48 = Set AttributeInitCategoryA(execute=@n43.Wood, AttributeInitCategoryA=ResourceWood) -> (Output_Get=None)
   n49 = Set SelectedResource(execute=@n2.NewEnumerator1, SelectedResource=Stone) -> (Output_Get=Wood)
   n50 = Set SelectedResource(execute=@n2.NewEnumerator2, SelectedResource=Metal) -> (Output_Get=Wood)
  ]}
 Stone {nodes [
   Stone() {is=title K2Node_FunctionEntry}
   > Set Prop Resource(Resource=NewEnumerator1) {is=title K2Node_CallFunction}
  ]}
 Wood {nodes [
   Wood() {is=title K2Node_FunctionEntry}
   > Set Prop Resource(Resource=NewEnumerator0)
  ]}
]
variables []
