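# Editor actor-action utility for UEFN that converts selected FortStaticMeshActors into prop Blueprints and replaces them in the level.
# Wood, Stone, Metal and Convert To Prop feed one helper with resource and prop type.
# The helper creates or updates Blueprints under CustomProps, reparents them to building classes, then sets mesh, damageability and resource attributes on their defaults.
# Repeated meshes reuse one Blueprint, and spawned replacements keep the original transforms.
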EditorUtilityBlueprint /Game/Blueprints/Blutilities/UEFNToolkit/StaticMesh_ActorActions
graphs [
 Convert To Prop {nodes [
   n1 = Convert To Prop!() -> (then, ResourceType:E_AssetToProp_ResourceType Enum=NewEnumerator0, PropType:E_AssetToProp_PropType Enum=NewEnumerator0) {is=title K2Node_FunctionEntry}
   > Convert To Prop Helper!(
    execute
    self
    ResourceType:E_AssetToProp_ResourceType Enum=@n1
    PropType:E_AssetToProp_PropType Enum=@n1.PropType
    isBuildingProp:bool=false
   ) -> (then)
  ]}
 Convert To Prop Helper {nodes [
   n1 = Convert To Prop Helper!() -> (
    then
    ResourceType:E_AssetToProp_ResourceType Enum=NewEnumerator0
    PropType:E_AssetToProp_PropType Enum=NewEnumerator0
    isBuildingProp:bool=~
   ) {is=title K2Node_FunctionEntry}
   > CancelTransaction!(execute, Index:int=0) -> (then)
   > Set Can Be Damaged(Can Be Damaged:bool=false) -> (Output_Get:bool)
   > n2 = Switch on E_AssetToProp_ResourceType!(execute, Selection:E_AssetToProp_ResourceType Enum=@n1) -> (NewEnumerator0:Exec, NewEnumerator1:Exec, NewEnumerator2:Exec) {is=title K2Node_SwitchEnum}
   > n3 = Set SelectedResourceType(SelectedResourceType:EFortResourceType Enum=Wood) -> (Output_Get:EFortResourceType Enum=Wood)
   Set SelectedPropType(execute=[@n107.then, @n108.then, @n3.then], SelectedPropType:E_AssetToProp_PropType Enum=@n1.PropType) -> (Output_Get:E_AssetToProp_PropType Enum=NewEnumerator1)
   > n4 = GetSelectionSet!(execute) -> (then, ReturnValue:Array of Actor Object References)
   > n5 = For Each Loop(Array:Array of Object References=@n4) -> (Array Element:Object Reference)
   n6 = Length(TargetArray=Get NewActors() -> (NewActors:Array of Actor Object References))
   PrintString(
    execute=@n5.Completed
    InString=Append!(A="Done converting ", B="To String (Integer)"(InInt=@n6), C:String=" mesh actors to props") -> (ReturnValue)
    bPrintToScreen=false
   )
   > SetSelectedLevelActors(self=Get EditorActorSubsystem(), ActorsToSelect=Get NewActors())
   > n7 = For Each Loop(Array=Get NewAssets() -> (NewAssets:Array of Object References))
   SyncBrowserToObjects(execute=@n7.Completed, AssetPaths=Get NewAssetsPaths() -> (NewAssetsPaths:Array of Strings))
   n8 = GetPathNameForLoadedAsset(execute=@n7.LoopBody, LoadedAsset=@n7)
   > Add(TargetArray=Get NewAssetsPaths(), NewItem=@n8) {is=title K2Node_CallArrayFunction}
   n9 = Cast To FortStaticMeshActor!(execute=@n5.LoopBody, Object=@n5) -> (then, CastFailed, AsFort Static Mesh Actor:Fort Static Mesh Actor Object Reference) {is=title K2Node_DynamicCast}
   > Set FortMeshActor(FortMeshActor:Fort Static Mesh Actor Object Reference=@n9) -> (Output_Get:Fort Static Mesh Actor Object Reference)
   n10 = Get StaticMesh(self=Get StaticMeshComponent(self=@n9))
   n11 = Find Item!(
    TargetArray:Array of Static Mesh Object References=Get MeshArray() -> (MeshArray:Array of Static Mesh Object References)
    ItemToFind:"Static Mesh Object Reference (by ref)"=@n10
   ) -> (ReturnValue)
   > n12 = Branch(Condition="Equal (Integer)"(A=@n11, B=-1))
   > PrintString(InString=No duplicates found, bPrintToScreen=false)
   > n13 = Set StaticMesh(StaticMesh:Static Mesh Object Reference=@n10) -> (Output_Get:Static Mesh Object Reference)
   > Add Unique(TargetArray:Array of Static Mesh Object References=Get MeshArray(), NewItem:"Static Mesh Object Reference (by ref)"=@n13)
   > n14 = GetPathNameForLoadedAsset(LoadedAsset=@n13)
   > n15 = GetAssetTools()
   n16 = Get PropNameOverride() -> (PropNameOverride:String)
   > n17 = Branch(Condition=IsEmpty(InString=@n16))
   n18 = GetDisplayName(Object=@n5)
   n19 = Get PropNamePrefix() -> (PropNamePrefix:String)
   > n20 = Set FinalPropName(FinalPropName:String=Replace(SourceString=Append!(A=@n19, B=@n18) -> (ReturnValue), From=" ")) -> (Output_Get:String)
   n21 = Replace(SourceString=Append!(A=/, B=GetProjectRootAssetDirectory!() -> (ReturnValue:String), C:String=/CustomProps/) -> (ReturnValue), From=" ")
   n22 = Get FinalPropName() -> (FinalPropName:String)
   n23 = Append!(A=@n21, B=@n22) -> (ReturnValue)
   n24 = DoesAssetExist(execute=[@n20.then, @n103.then], AssetPath=@n23)
   > n25 = Branch(Condition=@n24)
   > n26 = LoadAsset(AssetPath=@n23)
   > Set AssetObject(AssetObject:Object Reference=@n26) -> (Output_Get:Object Reference)
   > n27 = Show Suppressable Warning Dialog!(
    execute
    Title:"Text (by ref)"=MakeLiteralText(Value="LOCTEXT(\"Overwrite existing Prop asset ?\")")
    Message:"Text (by ref)"="To Text (String)"(InString=Append!(A="The prop asset \"", B=@n22, C:String="\" already exists, do you want to update it ?") -> (ReturnValue))
    InIniSettingName:String
    InIniSettingFileNameOverride:String
    bDefaultValue:bool=true
   ) -> (then, ReturnValue:bool)
   > n28 = Branch(Condition=@n27)
   > n29 = PrintString(
    InString=Append!(A="BuildingProp Asset already exists at this path, updating with new properties", B=@n23) -> (ReturnValue)
    bPrintToScreen=false
   )
   n30 = MakeSoftClassPath(PathString=/Script/Engine.Blueprint)
   n31 = CreateAsset(
    execute=@n25.else
    self=@n15
    AssetName=@n22
    PackagePath=@n21
    AssetClass=Resolve Soft Reference!(Input:Object Soft Class Reference=To Soft Class Reference(SoftClassPath=@n30)) -> (Output:Object Class Reference)
   )
   n32 = Set AssetObject(execute=@n109.else, AssetObject=@n31)
   n33 = GetBlueprintAsset(execute=[@n32.then, @n29.then], Object=@n32)
   > n34 = Branch(Condition=@n1.isBuildingProp)
   > n35 = ReparentBlueprint(Blueprint=@n33, NewParentClass=&BuildingProp)
   n36 = Gather Subobject Data For Blueprint(execute=[@n100.then, @n102.then, @n99.then, @n101.then, @n35.then], self=Get SubobjectDataSubsystem(), Context=@n33)
   > n37 = GetObjectForBlueprint!(
    execute
    Data:"Subobject Data Structure (by ref)"=GetData(DataHandle="Get (a copy)"(Array:Array of Subobject Data Handle Structures=@n36) -> (Output:Subobject Data Handle Structure))
    Blueprint:Blueprint Object Reference=@n33
   ) -> (then, ReturnValue:Object Reference)
   > n38 = Branch(Condition=@n1.isBuildingProp)
   > n39 = Cast To BuildingProp!(execute, Object=@n37) -> (then, CastFailed, AsBuilding Prop:Building Prop Object Reference) {is=title K2Node_DynamicCast}
   > Set BuildingPropAsset(BuildingPropAsset:Building Prop Object Reference=@n39) -> (Output_Get:Building Prop Object Reference)
   n40 = Get BuildingPropAsset() -> (BuildingPropAsset:Building Prop Object Reference)
   > Set StaticMesh!(execute, StaticMesh=Get StaticMesh!() -> (StaticMesh), self:Building Static Mesh Object Reference=@n40) -> (then, Output_Get)
   > SetCanBeDamaged!(execute, Actor:Actor Object Reference=@n40, bCanBeDamaged:bool=Get Can Be Damaged() -> (Can Be Damaged:bool)) -> (then)
   n41 = Get SelectedResourceType() -> (SelectedResourceType:EFortResourceType Enum=Wood)
   > SetResourceType!(execute, self:Building Static Mesh Object Reference=@n40, NewResourceType:EFortResourceType Enum=@n41) -> (then)
   > n42 = Switch on EFortResourceType!(execute, Selection:EFortResourceType Enum=@n41) -> (Wood:Exec, Stone:Exec, Metal:Exec, Permanite:Exec, GoldCurrency:Exec, Ingredient:Exec, None:Exec) {is=title K2Node_SwitchEnum}
   n43 = Set AttributeInitCategory(execute=[@n42.GoldCurrency, @n42.Ingredient, @n42.None, @n42.Permanite], AttributeInitCategory:Name=ResourceWood) -> (Output_Get:Name=None)
   n44 = Make Fort Attribute Initialization Key!(AttributeInitCategory:Name=Get AttributeInitCategory() -> (AttributeInitCategory:Name=None), AttributeInitSubCategory:Name=2) -> (FortAttributeInitializationKey:Fort Attribute Initialization Key Structure) {is=title K2Node_MakeStruct}
   n45 = SetEditorProperty(
    execute=[@n56.then, @n55.then, @n54.then, @n43.then]
    Object=Get BuildingPropAsset()
    PropertyName=AttributeInitKeys
    PropertyValue:Array of Fort Attribute Initialization Key Structures=Make Array!("[0]":Fort Attribute Initialization Key Structure=@n44, "[1]":Fort Attribute Initialization Key Structure=@n44) -> (Array:Array of Fort Attribute Initialization Key Structures)
   )
   n46 = GetBlueprintAsset(execute=[@n94.then, @n45.then, @n64.then, @n74.then, @n84.then], Object=Get AssetObject() -> (AssetObject:Object Reference))
   > Add Unique(TargetArray:Array of Object References=Get NewAssets(), NewItem:"Object Reference (by ref)"=@n46)
   > CompileBlueprint!(execute, Blueprint:Blueprint Object Reference=@n46) -> (then, ReturnValue:bool)
   > SaveLoadedAsset(AssetToSave=@n46, bOnlyIfIsDirty=true)
   > n47 = GeneratedClass!(execute, BlueprintObj:Blueprint Object Reference=@n46) -> (then, ReturnValue:Object Class Reference)
   > n48 = Set PropBlueprintGeneratedClass(PropBlueprintGeneratedClass:Object Class Reference=@n47) -> (Output_Get:Object Class Reference)
   n49 = Cast To Actor Class!(
    execute:Exec=[@n48.then, @n106.then]
    Class:Object Class Reference=Get PropBlueprintGeneratedClass() -> (PropBlueprintGeneratedClass:Object Class Reference)
   ) -> (then:Exec, CastFailed:Exec, AsActor:Actor Class Reference) {is=title K2Node_ClassDynamicCast}
   > n50 = Set BuildingPropClass(BuildingPropClass:Actor Class Reference=@n49) -> (Output_Get:Actor Class Reference)
   > BeginTransaction(Context=ActorToProp, Description="LOCTEXT(\"Replace selected Mesh Actor(s) with created Prop(s)\")")
   n51 = Get FortMeshActor() -> (FortMeshActor:Fort Static Mesh Actor Object Reference)
   n52 = Get Actor Transform!(self=@n51) -> (ReturnValue_Location:Vector, ReturnValue_Rotation:Rotator, ReturnValue_Scale:Vector)
   > n53 = SpawnActorFromClass!(
    execute
    self:Editor Actor Subsystem Object Reference=Get EditorActorSubsystem()
    ActorClass:Actor Class Reference=@n50
    Location:Vector=@n52
    Rotation:Rotator=@n52.ReturnValue_Rotation
    bTransient:bool
   ) -> (then, ReturnValue:Actor Object Reference)
   > SetActorScale3D!(execute, self:Actor Object Reference=@n53, NewScale3D:Vector=@n52.ReturnValue_Scale) -> (then)
   > Add(TargetArray:Array of Actor Object References=Get NewActors(), NewItem:"Actor Object Reference (by ref)"=@n53)
   > DestroyActor!(execute, self:Editor Actor Subsystem Object Reference=Get EditorActorSubsystem(), ActorToDestroy:Actor Object Reference=@n51) -> (then, ReturnValue:bool)
   > EndTransaction()
   n54 = Set AttributeInitCategory(execute=@n42.Metal, AttributeInitCategory=ResourceMetal) -> (Output_Get=None)
   n55 = Set AttributeInitCategory(execute=@n42.Stone, AttributeInitCategory=ResourceStone) -> (Output_Get=None)
   n56 = Set AttributeInitCategory(execute=@n42.Wood, AttributeInitCategory=ResourceWood) -> (Output_Get=None)
   PrintString(execute=@n39.CastFailed, InString=Cast to BuildingProp failed, TextColor=(R=1,G=0,B=0,A=1))
   n57 = Switch on E_AssetToProp_PropType!(
    execute=@n38.else
    Selection:E_AssetToProp_PropType Enum=Get SelectedPropType() -> (SelectedPropType:E_AssetToProp_PropType Enum=NewEnumerator1)
   ) -> (NewEnumerator1:Exec, NewEnumerator2:Exec, NewEnumerator3:Exec, NewEnumerator4:Exec) {is=title K2Node_SwitchEnum}
   > n58 = Cast To BuildingPropWall!(execute, Object=@n37) -> (then, CastFailed, AsBuilding Prop Wall:Building Prop Wall Object Reference) {is=title K2Node_DynamicCast}
   > Set BuildingPropWallAsset(BuildingPropWallAsset:Building Prop Wall Object Reference=@n58) -> (Output_Get:Building Prop Wall Object Reference)
   n59 = Get BuildingPropWallAsset() -> (BuildingPropWallAsset:Building Prop Wall Object Reference)
   > Set StaticMesh(StaticMesh=Get StaticMesh(), self=@n59)
   > SetCanBeDamaged(Actor=@n59, bCanBeDamaged=Get Can Be Damaged())
   n60 = Get SelectedResourceType() -> (SelectedResourceType=Wood)
   > SetResourceType(self=@n59, NewResourceType=@n60)
   > n61 = Switch on EFortResourceType(Selection=@n60)
   n62 = Set AttributeInitCategory(execute=[@n61.GoldCurrency, @n61.Ingredient, @n61.None, @n61.Permanite], AttributeInitCategory=ResourceWood) -> (Output_Get=None)
   n63 = Make Fort Attribute Initialization Key(AttributeInitCategory=Get AttributeInitCategory() -> (AttributeInitCategory=None), AttributeInitSubCategory=2)
   n64 = SetEditorProperty(
    execute=[@n67.then, @n66.then, @n65.then, @n62.then]
    Object=Get BuildingPropWallAsset()
    PropertyName=AttributeInitKeys
    PropertyValue=Make Array("[0]"=@n63, "[1]"=@n63)
   )
   n65 = Set AttributeInitCategory(execute=@n61.Metal, AttributeInitCategory=ResourceMetal) -> (Output_Get=None)
   n66 = Set AttributeInitCategory(execute=@n61.Stone, AttributeInitCategory=ResourceStone) -> (Output_Get=None)
   n67 = Set AttributeInitCategory(execute=@n61.Wood, AttributeInitCategory=ResourceWood) -> (Output_Get=None)
   PrintString(execute=@n58.CastFailed, InString=Cast to BuildingPropWall failed, TextColor=(R=1,G=0,B=0,A=1))
   n68 = Cast To BuildingFloor!(execute=@n57.NewEnumerator2, Object=@n37) -> (then, CastFailed, AsBuilding Floor:Building Floor Object Reference) {is=title K2Node_DynamicCast}
   > Set BuildingPropFloorAsset(BuildingPropFloorAsset:Building Floor Object Reference=@n68) -> (Output_Get:Building Floor Object Reference)
   n69 = Get BuildingPropFloorAsset() -> (BuildingPropFloorAsset:Building Floor Object Reference)
   > Set StaticMesh(StaticMesh=Get StaticMesh(), self=@n69)
   > SetCanBeDamaged(Actor=@n69, bCanBeDamaged=Get Can Be Damaged())
   n70 = Get SelectedResourceType() -> (SelectedResourceType=Wood)
   > SetResourceType(self=@n69, NewResourceType=@n70)
   > n71 = Switch on EFortResourceType(Selection=@n70)
   n72 = Set AttributeInitCategory(execute=[@n71.GoldCurrency, @n71.Ingredient, @n71.None, @n71.Permanite], AttributeInitCategory=ResourceWood) -> (Output_Get=None)
   n73 = Make Fort Attribute Initialization Key(AttributeInitCategory=Get AttributeInitCategory() -> (AttributeInitCategory=None), AttributeInitSubCategory=2)
   n74 = SetEditorProperty(
    execute=[@n77.then, @n76.then, @n75.then, @n72.then]
    Object=Get BuildingPropFloorAsset()
    PropertyName=AttributeInitKeys
    PropertyValue=Make Array("[0]"=@n73, "[1]"=@n73)
   )
   n75 = Set AttributeInitCategory(execute=@n71.Metal, AttributeInitCategory=ResourceMetal) -> (Output_Get=None)
   n76 = Set AttributeInitCategory(execute=@n71.Stone, AttributeInitCategory=ResourceStone) -> (Output_Get=None)
   n77 = Set AttributeInitCategory(execute=@n71.Wood, AttributeInitCategory=ResourceWood) -> (Output_Get=None)
   PrintString(execute=@n68.CastFailed, InString=Cast to BuildingFloor failed, TextColor=(R=1,G=0,B=0,A=1))
   n78 = Cast To BuildingPropCorner!(execute=@n57.NewEnumerator3, Object=@n37) -> (then, CastFailed, AsBuilding Prop Corner:Building Prop Corner Object Reference) {is=title K2Node_DynamicCast}
   > Set BuildingPropCornerAsset(BuildingPropCornerAsset:Building Prop Corner Object Reference=@n78) -> (Output_Get:Building Prop Corner Object Reference)
   n79 = Get BuildingPropCornerAsset() -> (BuildingPropCornerAsset:Building Prop Corner Object Reference)
   > Set StaticMesh(StaticMesh=Get StaticMesh(), self=@n79)
   > SetCanBeDamaged(Actor=@n79, bCanBeDamaged=Get Can Be Damaged())
   n80 = Get SelectedResourceType() -> (SelectedResourceType=Wood)
   > SetResourceType(self=@n79, NewResourceType=@n80)
   > n81 = Switch on EFortResourceType(Selection=@n80)
   n82 = Set AttributeInitCategory(execute=[@n81.GoldCurrency, @n81.Ingredient, @n81.None, @n81.Permanite], AttributeInitCategory=ResourceWood) -> (Output_Get=None)
   n83 = Make Fort Attribute Initialization Key(AttributeInitCategory=Get AttributeInitCategory() -> (AttributeInitCategory=None), AttributeInitSubCategory=2)
   n84 = SetEditorProperty(
    execute=[@n87.then, @n86.then, @n85.then, @n82.then]
    Object=Get BuildingPropCornerAsset()
    PropertyName=AttributeInitKeys
    PropertyValue=Make Array("[0]"=@n83, "[1]"=@n83)
   )
   n85 = Set AttributeInitCategory(execute=@n81.Metal, AttributeInitCategory=ResourceMetal) -> (Output_Get=None)
   n86 = Set AttributeInitCategory(execute=@n81.Stone, AttributeInitCategory=ResourceStone) -> (Output_Get=None)
   n87 = Set AttributeInitCategory(execute=@n81.Wood, AttributeInitCategory=ResourceWood) -> (Output_Get=None)
   PrintString(execute=@n78.CastFailed, InString=Cast to BuildingPropCorner failed, TextColor=(R=1,G=0,B=0,A=1))
   n88 = Cast To BuildingRoof!(execute=@n57.NewEnumerator4, Object=@n37) -> (then, CastFailed, AsBuilding Roof:Building Roof Object Reference) {is=title K2Node_DynamicCast}
   > Set BuildingRoofAsset(BuildingRoofAsset:Building Roof Object Reference=@n88) -> (Output_Get:Building Roof Object Reference)
   n89 = Get BuildingRoofAsset() -> (BuildingRoofAsset:Building Roof Object Reference)
   > Set StaticMesh(StaticMesh=Get StaticMesh(), self=@n89)
   > SetCanBeDamaged(Actor=@n89, bCanBeDamaged=Get Can Be Damaged())
   n90 = Get SelectedResourceType() -> (SelectedResourceType=Wood)
   > SetResourceType(self=@n89, NewResourceType=@n90)
   > n91 = Switch on EFortResourceType(Selection=@n90)
   n92 = Set AttributeInitCategory(execute=[@n91.GoldCurrency, @n91.Ingredient, @n91.None, @n91.Permanite], AttributeInitCategory=ResourceWood) -> (Output_Get=None)
   n93 = Make Fort Attribute Initialization Key(AttributeInitCategory=Get AttributeInitCategory() -> (AttributeInitCategory=None), AttributeInitSubCategory=2)
   n94 = SetEditorProperty(
    execute=[@n92.then, @n96.then, @n95.then, @n97.then]
    Object=Get BuildingRoofAsset()
    PropertyName=AttributeInitKeys
    PropertyValue=Make Array("[0]"=@n93, "[1]"=@n93)
   )
   n95 = Set AttributeInitCategory(execute=@n91.Metal, AttributeInitCategory=ResourceMetal) -> (Output_Get=None)
   n96 = Set AttributeInitCategory(execute=@n91.Stone, AttributeInitCategory=ResourceStone) -> (Output_Get=None)
   n97 = Set AttributeInitCategory(execute=@n91.Wood, AttributeInitCategory=ResourceWood) -> (Output_Get=None)
   PrintString(execute=@n88.CastFailed, InString=Cast to BuildingRoof failed, TextColor=(R=1,G=0,B=0,A=1))
   n98 = Switch on E_AssetToProp_PropType(execute=@n34.else, Selection=Get SelectedPropType() -> (SelectedPropType=NewEnumerator1))
   > n99 = ReparentBlueprint(Blueprint=@n33, NewParentClass=/Script/FortniteGame.BuildingPropWall)
   n100 = ReparentBlueprint(execute=@n98.NewEnumerator2, Blueprint=@n33, NewParentClass=&BuildingFloor)
   n101 = ReparentBlueprint(execute=@n98.NewEnumerator3, Blueprint=@n33, NewParentClass=/Script/FortniteGame.BuildingPropCorner)
   n102 = ReparentBlueprint(execute=@n98.NewEnumerator4, Blueprint=@n33, NewParentClass=&BuildingRoof)
   PrintString(execute=[@n109.then, @n28.else], InString=Prop asset creation failed)
   n103 = Set FinalPropName(execute=@n17.else, FinalPropName=Replace(SourceString=@n16, From=" "))
   PrintString(execute=@n12.else, InString="Duplicate found, using original reference instead", bPrintToScreen=false)
   > Set OGMeshID(OGMeshID:int=@n11) -> (Output_Get:int=0)
   > n104 = GetBlueprintAsset(Object="Get (a copy)"(Array:Array of Object References=Get NewAssets(), Dimension 1=Get OGMeshID() -> (OGMeshID:int=0)) -> (Output:Object Reference))
   > SaveLoadedAsset(AssetToSave=@n104, bOnlyIfIsDirty=true)
   > n105 = GeneratedClass(BlueprintObj=@n104)
   > n106 = Set PropBlueprintGeneratedClass(PropBlueprintGeneratedClass=@n105)
   PrintString(execute=@n9.CastFailed, InString=Selected Actor is not a FortStaticMesh, bPrintToScreen=false)
   n107 = Set SelectedResourceType(execute=@n2.NewEnumerator1, SelectedResourceType=Stone) -> (Output_Get=Wood)
   n108 = Set SelectedResourceType(execute=@n2.NewEnumerator2, SelectedResourceType=Metal) -> (Output_Get=Wood)
   Append(A=Split!(InPath:String=@n14) -> (PathPart:String, FilenamePart:String, ExtensionPart:String), B=/)
   n109 = Branch(execute=@n31.then, Condition="Equal (Object)"(A=@n31))
  ]}
 EventGraph {links [], nodes []}
 Metal {nodes [
   Metal() {is=title K2Node_FunctionEntry}
   > Convert To Prop Helper(ResourceType=NewEnumerator2, PropType=NewEnumerator1, isBuildingProp=true) {is=title K2Node_CallFunction}
  ]}
 Stone {nodes [
   Stone() {is=title K2Node_FunctionEntry}
   > Convert To Prop Helper(ResourceType=NewEnumerator1, PropType=NewEnumerator1, isBuildingProp=true)
  ]}
 Wood {nodes [
   Wood() {is=title K2Node_FunctionEntry}
   > Convert To Prop Helper(ResourceType=NewEnumerator0, PropType=NewEnumerator1, isBuildingProp=true)
  ]}
]
variables []
